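# Behavior tree injected into the Cerberus AI (AIGods) to perform its melee slash attack.
# A root Selector, bound to the BB_MeleeSlash blackboard, holds a single TriggerAbility task.
# That task executes the gameplay ability tagged Ability.AIGods.Cerberus.MeleeSlash, with no decorators or services gating it.
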BehaviorTree /AIGods_Cerberus/BTInjection/BT_MeleeSlash
blackboard=&/BB_MeleeSlash
root {
 name=Selector
 class=BTComposite_Selector
 children [TriggerAbility {
   class=FortBTTask_ExecuteGameplayAbility
   decorators []
   properties {nodeName=TriggerAbility, gameplayAbilityTagContainer {defaultValue {gameplayTags [{tagName=Ability.AIGods.Cerberus.MeleeSlash}]}}}
  }]
 decorators []
 properties {
  children [{childComposite=null, childTask=&self:FortBTTask_ExecuteGameplayAbility_4, decorators []}]
  services []
 }
 services []
}
